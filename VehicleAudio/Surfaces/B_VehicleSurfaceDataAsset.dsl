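; Vehicle sound data asset that selects the sound to play for a physical surface or for an airborne vehicle.
; When airborne it returns the in-air sound if valid; otherwise it looks up a data table row named after the surface enum and picks a random sound.
; Failed ground lookups fall back to the default sound, with debug prints for a missing table or default sound.
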(fn GetSoundForSurface
  (params (InSurface :type EPhysicalSurface%20Enum :default "SurfaceType_Default") (InAir :type Boolean))
  (bind _surfacesound (CallFunction|GetSoundforSurface InSurface InAir))
  (return _surfacesound))

(fn GetSoundForSurface
  (params (Surface :type EPhysicalSurface%20Enum :default "SurfaceType_Default") (InAir :type Boolean))
  (bind _surfacedata (Variables|VehicleSoundSurfaceData|GetSurfaceData))
  (bind _inairsound (Variables|VehicleSoundSurfaceData|GetInAirSound))
  (bind _defaultsound (Variables|VehicleSoundSurfaceData|GetDefaultSound))
  (if InAir
    (Utilities|IsValid _inairsound
      (:Is%20Valid
        (return _inairsound)))
    (else
      (Utilities|IsValid _surfacedata
        (:Is%20Valid
          (bind _returnvalue (Utilities|GetDataTableRow _surfacedata (Utilities|String|StringToName (Utilities|String|EnumtoString Surface)))
            (:then
              (Utilities|IsValid (Utilities|Array|RandomArrayItem (Utilities|Struct|BreakStrutAudioSurface _returnvalue))
                (:Is%20Valid
                  (return (Utilities|Array|RandomArrayItem (Utilities|Struct|BreakStrutAudioSurface _returnvalue))))))
            (:RowNotFound
              _)))
        (:Is%20Not%20Valid
          (Development|PrintString 0 "No Table Selected!")))
      (Utilities|IsValid _defaultsound
        (:Is%20Valid
          (return _defaultsound))
        (:Is%20Not%20Valid
          (Development|PrintString 0 "No Default Sound Selected!"))))))
